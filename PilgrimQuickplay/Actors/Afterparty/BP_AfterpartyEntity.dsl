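; BP_AfterpartyEntity is a Blueprint actor for the afterparty feature in Pilgrim Quickplay.
; It declares a custom event, ActivateAfterpartyEffect, with no logic attached here.
; Its construction script is also empty, so the event reads as a hook, likely for child Blueprints to implement their effects.
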(event Custom|ActivateAfterpartyEffect)

(fn ConstructionScript)
